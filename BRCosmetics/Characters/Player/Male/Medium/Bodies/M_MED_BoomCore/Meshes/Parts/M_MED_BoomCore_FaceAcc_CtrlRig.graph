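# Control Rig for the BoomCore male face accessory mesh that corrects back-hair and sideburn bones as the head moves.
# Four spherical pose readers measure head orientation against neck_01, spine_05 or neck_02.
# Each reader's output weights rotation offsets on five hair master bones, propagated to their child chains.
# Two parent-constraint chains tying lower hair bones to head, neck and spine are unconnected to execution, so they never run.
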ControlRigBlueprint /BRCosmetics/Characters/Player/Male/Medium/Bodies/M_MED_BoomCore/Meshes/Parts/M_MED_BoomCore_FaceAcc_CtrlRig
graphs [
 RigVMFunctionLibrary {links [], nodes []}
 RigVMModel {nodes [
   &RigUnit_ParentConstraint(
    AdvancedSettings=(InterpolationType=Average,RotationOrderForFilter=XZY)
    Child=(Type=Bone,Name="dyn_bk_l_hair_02")
    ExecutePin=()
    Parents=((Item=(Type=Bone,Name="neck_02"),Weight=1))
    Weight=1
   )
   > &RigUnit_ParentConstraint(
    AdvancedSettings=(InterpolationType=Average,RotationOrderForFilter=XZY)
    Child=(Type=Bone,Name="dyn_bk_hair_02")
    Parents=((Item=(Type=Bone,Name="neck_02"),Weight=1))
    Weight=1
   )
   > &RigUnit_ParentConstraint(
    AdvancedSettings=(InterpolationType=Average,RotationOrderForFilter=XZY)
    Child=(Type=Bone,Name="dyn_bk_r_hair_02")
    Parents=((Item=(Type=Bone,Name="neck_02"),Weight=1))
    Weight=1
   )
   > &RigUnit_ParentConstraint(
    AdvancedSettings=(InterpolationType=Average,RotationOrderForFilter=XZY)
    Child=(Type=Bone,Name="dyn_bk_l_hair_03")
    Parents=((Item=(Type=Bone,Name="neck_01"),Weight=1))
    Weight=1
   )
   > &RigUnit_ParentConstraint(
    AdvancedSettings=(InterpolationType=Average,RotationOrderForFilter=XZY)
    Child=(Type=Bone,Name="dyn_bk_hair_03")
    Parents=((Item=(Type=Bone,Name="neck_01"),Weight=1))
    Weight=1
   )
   > &RigUnit_ParentConstraint(
    AdvancedSettings=(InterpolationType=Average,RotationOrderForFilter=XZY)
    Child=(Type=Bone,Name="dyn_bk_r_hair_03")
    Parents=((Item=(Type=Bone,Name="neck_01"),Weight=1))
    Weight=1
   )
   > &RigUnit_ParentConstraint(
    AdvancedSettings=(InterpolationType=Average,RotationOrderForFilter=XZY)
    Child=(Type=Bone,Name="dyn_bk_l_hair_04")
    Parents=((Item=(Type=Bone,Name="spine_05"),Weight=1))
    Weight=1
   )
   > &RigUnit_ParentConstraint(
    AdvancedSettings=(InterpolationType=Average,RotationOrderForFilter=XZY)
    Child=(Type=Bone,Name="dyn_bk_hair_04")
    Parents=((Item=(Type=Bone,Name="spine_05"),Weight=1))
    Weight=1
   )
   > &RigUnit_ParentConstraint(
    AdvancedSettings=(InterpolationType=Average,RotationOrderForFilter=XZY)
    Child=(Type=Bone,Name="dyn_bk_r_hair_04")
    Parents=((Item=(Type=Bone,Name="spine_05"),Weight=1))
    Weight=1
   )
   &RigUnit_ParentConstraint(
    AdvancedSettings=(InterpolationType=Average,RotationOrderForFilter=XZY)
    Child=(Type=Bone,Name="dyn_l_sideburn_hair_02")
    ExecutePin=()
    Parents=((Item=(Type=Bone,Name="head"),Weight=1))
    Weight=1
   )
   > &RigUnit_ParentConstraint(
    AdvancedSettings=(InterpolationType=Average,RotationOrderForFilter=XZY)
    Child=(Type=Bone,Name="dyn_r_sideburn_hair_02")
    Parents=((Item=(Type=Bone,Name="head"),Weight=1))
    Weight=1
   )
   > &RigUnit_ParentConstraint(
    AdvancedSettings=(InterpolationType=Average,RotationOrderForFilter=XZY)
    Child=(Type=Bone,Name="dyn_l_sideburn_hair_03")
    Parents=((Item=(Type=Bone,Name="neck_01"),Weight=1))
    Weight=1
   )
   > &RigUnit_ParentConstraint(
    AdvancedSettings=(InterpolationType=Average,RotationOrderForFilter=XZY)
    Child=(Type=Bone,Name="dyn_r_sideburn_hair_03")
    Parents=((Item=(Type=Bone,Name="neck_01"),Weight=1))
    Weight=1
   )
   > &RigUnit_ParentConstraint(
    AdvancedSettings=(InterpolationType=Average,RotationOrderForFilter=XZY)
    Child=(Type=Bone,Name="dyn_l_sideburn_hair_04")
    Parents=((Item=(Type=Bone,Name="spine_05"),Weight=1))
    Weight=1
   )
   > &RigUnit_ParentConstraint(
    AdvancedSettings=(InterpolationType=Average,RotationOrderForFilter=XZY)
    Child=(Type=Bone,Name="dyn_r_sideburn_hair_04")
    Parents=((Item=(Type=Bone,Name="spine_05"),Weight=1))
    Weight=1
   )
   RigUnit_BeginExecution = &RigUnit_BeginExecution()
   n1 = &RigUnit_SphericalPoseReader(
    ActiveRegionScaleFactors=(PositiveWidth=1,NegativeWidth=1,PositiveHeight=1,NegativeHeight=1)
    ActiveRegionSize=0.1
    Debug=(bDrawDebug=false,DebugScale=25,DebugSegments=20,DebugThickness=0.25)
    DriverAxis=(X=0,Y=-1,Z=0)
    DriverItem=(Type=Bone,Name="head")
    FalloffRegionScaleFactors=(PositiveWidth=2,NegativeWidth=2,PositiveHeight=2,NegativeHeight=2)
    FalloffSize=0.2
    FlipWidthScaling=false
    OptionalParentItem=(Type=Bone,Name="neck_01")
    RotationOffset=(X=-90,Y=0,Z=-55)
    ExecutePin=@RigUnit_BeginExecution.ExecutePin
   )
   > &RigUnit_OffsetTransformForItem(
    Item=(Type=Bone,Name="dyn_bk_l_hair_MSTR")
    OffsetTransform=(Rotation=(X=-0.25,Y=-0.433,Z=-0.433,W=0.75),Translation=(X=0,Y=0,Z=0),Scale3D=(X=1,Y=1,Z=1))
    bPropagateToChildren=True
    Weight=@n1.OutputParam
   )
   > &RigUnit_OffsetTransformForItem(
    Item=(Type=Bone,Name="dyn_bk_hair_MSTR")
    OffsetTransform=(Rotation=(X=0,Y=-0.5736,Z=0,W=0.8192),Translation=(X=0,Y=0,Z=0),Scale3D=(X=1,Y=1,Z=1))
    bPropagateToChildren=True
    Weight=@n1.OutputParam
   )
   > &RigUnit_OffsetTransformForItem(
    Item=(Type=Bone,Name="dyn_bk_r_hair_MSTR")
    OffsetTransform=(Rotation=(X=-0.25,Y=-0.433,Z=-0.433,W=0.75),Translation=(X=0,Y=0,Z=0),Scale3D=(X=1,Y=1,Z=1))
    bPropagateToChildren=True
    Weight=@n1.OutputParam
   )
   > &RigUnit_OffsetTransformForItem(
    Item=(Type=Bone,Name="dyn_l_sideburn_hair_MSTR")
    OffsetTransform=(Rotation=(X=0,Y=0,Z=-0.2588,W=0.9659),Translation=(X=0,Y=0,Z=0),Scale3D=(X=1,Y=1,Z=1))
    bPropagateToChildren=True
    Weight=@n1.OutputParam
   )
   > &RigUnit_OffsetTransformForItem(
    Item=(Type=Bone,Name="dyn_r_sideburn_hair_MSTR")
    OffsetTransform=(Rotation=(X=0,Y=0,Z=-0.2588,W=0.9659),Translation=(X=0,Y=0,Z=0),Scale3D=(X=1,Y=1,Z=1))
    bPropagateToChildren=True
    Weight=@n1.OutputParam
   )
   > n2 = &RigUnit_SphericalPoseReader(
    ActiveRegionScaleFactors=(PositiveWidth=0.25,NegativeWidth=0.25,PositiveHeight=1,NegativeHeight=1)
    ActiveRegionSize=0.1
    Debug=(bDrawDebug=false,DebugScale=25,DebugSegments=20,DebugThickness=0.25)
    DriverAxis=(X=0,Y=1,Z=0)
    DriverItem=(Type=Bone,Name="head")
    FalloffRegionScaleFactors=(PositiveWidth=0.75,NegativeWidth=0.5,PositiveHeight=4,NegativeHeight=4)
    FalloffSize=0.2
    OptionalParentItem=(Type=Bone,Name="spine_05")
    RotationOffset=(X=90,Y=0,Z=25)
   )
   > &RigUnit_OffsetTransformForItem(
    Item=(Type=Bone,Name="dyn_bk_l_hair_MSTR")
    OffsetTransform=(Rotation=(X=-0.017,Y=0.1294,Z=0.1294,W=0.983),Translation=(X=0,Y=0,Z=0),Scale3D=(X=1,Y=1,Z=1))
    bPropagateToChildren=True
    Weight=@n2.OutputParam
   )
   > &RigUnit_OffsetTransformForItem(
    Item=(Type=Bone,Name="dyn_bk_hair_MSTR")
    OffsetTransform=(Rotation=(X=0,Y=0.1305,Z=0,W=0.9914),Translation=(X=0,Y=0,Z=0),Scale3D=(X=1,Y=1,Z=1))
    bPropagateToChildren=True
    Weight=@n2.OutputParam
   )
   > &RigUnit_OffsetTransformForItem(
    Item=(Type=Bone,Name="dyn_bk_r_hair_MSTR")
    OffsetTransform=(Rotation=(X=-0.017,Y=0.1294,Z=0.1294,W=0.983),Translation=(X=0,Y=0,Z=0),Scale3D=(X=1,Y=1,Z=1))
    bPropagateToChildren=True
    Weight=@n2.OutputParam
   )
   > &RigUnit_OffsetTransformForItem(
    Item=(Type=Bone,Name="dyn_l_sideburn_hair_MSTR")
    OffsetTransform=(Rotation=(X=0,Y=0,Z=0.2588,W=0.9659),Translation=(X=0,Y=0,Z=0),Scale3D=(X=1,Y=1,Z=1))
    bPropagateToChildren=True
    Weight=@n2.OutputParam
   )
   > &RigUnit_OffsetTransformForItem(
    Item=(Type=Bone,Name="dyn_r_sideburn_hair_MSTR")
    OffsetTransform=(Rotation=(X=0,Y=0,Z=0.2588,W=0.9659),Translation=(X=0,Y=0,Z=0),Scale3D=(X=1,Y=1,Z=1))
    bPropagateToChildren=True
    Weight=@n2.OutputParam
   )
   > n3 = &RigUnit_SphericalPoseReader(
    ActiveRegionScaleFactors=(PositiveWidth=1,NegativeWidth=0.5,PositiveHeight=1,NegativeHeight=1)
    ActiveRegionSize=0.1
    Debug=(bDrawDebug=false,DebugScale=25,DebugSegments=20,DebugThickness=0.25)
    DriverAxis=(X=0,Y=-1,Z=0)
    DriverItem=(Type=Bone,Name="head")
    FalloffRegionScaleFactors=(PositiveWidth=1,NegativeWidth=0.5,PositiveHeight=1,NegativeHeight=1)
    FalloffSize=0.2
    OptionalParentItem=(Type=Bone,Name="neck_02")
    RotationOffset=(X=-55,Y=0,Z=-10)
   )
   > &RigUnit_OffsetTransformForItem(
    Item=(Type=Bone,Name="dyn_bk_l_hair_MSTR")
    OffsetTransform=(Rotation=(X=0.017,Y=-0.1294,Z=0.1294,W=0.983),Translation=(X=0,Y=0,Z=0),Scale3D=(X=1,Y=1,Z=1))
    bPropagateToChildren=True
    Weight=@n3.OutputParam
   )
   > &RigUnit_OffsetTransformForItem(
    Item=(Type=Bone,Name="dyn_bk_hair_MSTR")
    OffsetTransform=(Rotation=(X=0,Y=-0.2588,Z=0,W=0.9659),Translation=(X=0,Y=0,Z=0),Scale3D=(X=1,Y=1,Z=1))
    bPropagateToChildren=True
    Weight=@n3.OutputParam
   )
   > &RigUnit_OffsetTransformForItem(
    Item=(Type=Bone,Name="dyn_bk_r_hair_MSTR")
    OffsetTransform=(Rotation=(X=-0.017,Y=-0.1294,Z=-0.1294,W=0.983),Translation=(X=0,Y=0,Z=0),Scale3D=(X=1,Y=1,Z=1))
    bPropagateToChildren=True
    Weight=@n3.OutputParam
   )
   > &RigUnit_OffsetTransformForItem(
    Item=(Type=Bone,Name="dyn_l_sideburn_hair_MSTR")
    OffsetTransform=(Rotation=(X=0,Y=0,Z=0.2588,W=0.9659),Translation=(X=0,Y=0,Z=0),Scale3D=(X=1,Y=1,Z=1))
    bPropagateToChildren=True
    Weight=@n3.OutputParam
   )
   > &RigUnit_OffsetTransformForItem(
    Item=(Type=Bone,Name="dyn_r_sideburn_hair_MSTR")
    OffsetTransform=(Rotation=(X=0,Y=0,Z=-0.1305,W=0.9914),Translation=(X=0,Y=0,Z=0),Scale3D=(X=1,Y=1,Z=1))
    bPropagateToChildren=True
    Weight=@n3.OutputParam
   )
   > n4 = &RigUnit_SphericalPoseReader(
    ActiveRegionScaleFactors=(PositiveWidth=1,NegativeWidth=0.5,PositiveHeight=1,NegativeHeight=1)
    ActiveRegionSize=0.1
    Debug=(bDrawDebug=false,DebugScale=25,DebugSegments=20,DebugThickness=0.25)
    DriverAxis=(X=0,Y=-1,Z=0)
    DriverItem=(Type=Bone,Name="head")
    FalloffRegionScaleFactors=(PositiveWidth=1,NegativeWidth=0.5,PositiveHeight=1,NegativeHeight=1)
    FalloffSize=0.2
    OptionalParentItem=(Type=Bone,Name="neck_02")
    RotationOffset=(X=235,Y=0,Z=-10)
   )
   > &RigUnit_OffsetTransformForItem(
    Item=(Type=Bone,Name="dyn_bk_l_hair_MSTR")
    OffsetTransform=(Rotation=(X=-0.017,Y=-0.1294,Z=-0.1294,W=0.983),Translation=(X=0,Y=0,Z=0),Scale3D=(X=1,Y=1,Z=1))
    bPropagateToChildren=True
    Weight=@n4.OutputParam
   )
   > &RigUnit_OffsetTransformForItem(
    Item=(Type=Bone,Name="dyn_bk_hair_MSTR")
    OffsetTransform=(Rotation=(X=0,Y=-0.2588,Z=0,W=0.9659),Translation=(X=0,Y=0,Z=0),Scale3D=(X=1,Y=1,Z=1))
    bPropagateToChildren=True
    Weight=@n4.OutputParam
   )
   > &RigUnit_OffsetTransformForItem(
    Item=(Type=Bone,Name="dyn_bk_r_hair_MSTR")
    OffsetTransform=(Rotation=(X=0.017,Y=-0.1294,Z=0.1294,W=0.983),Translation=(X=0,Y=0,Z=0),Scale3D=(X=1,Y=1,Z=1))
    bPropagateToChildren=True
    Weight=@n4.OutputParam
   )
   > &RigUnit_OffsetTransformForItem(
    Item=(Type=Bone,Name="dyn_l_sideburn_hair_MSTR")
    OffsetTransform=(Rotation=(X=0,Y=0,Z=-0.1305,W=0.9914),Translation=(X=0,Y=0,Z=0),Scale3D=(X=1,Y=1,Z=1))
    bPropagateToChildren=True
    Weight=@n4.OutputParam
   )
   > &RigUnit_OffsetTransformForItem(
    Item=(Type=Bone,Name="dyn_r_sideburn_hair_MSTR")
    OffsetTransform=(Rotation=(X=0,Y=0,Z=0.2588,W=0.9659),Translation=(X=0,Y=0,Z=0),Scale3D=(X=1,Y=1,Z=1))
    bPropagateToChildren=True
    Weight=@n4.OutputParam
   )
  ]}
]
hierarchy [
 root {loc=<0 0 0>, rot=<0 0 0 1>}
 pelvis {parent=root, loc=<0 -0.7071 93.1191>, rot=<0.0127 0.707 -0.0127 -0.707>}
 spine_01 {parent=pelvis, loc=<6.7494 -0.6388 0>, rot=<0 0 0.034 -0.9994>}
 spine_02 {parent=spine_01, loc=<7.2518 -0.035 0>, rot=<0 0 0.0238 -0.9997>}
 spine_03 {parent=spine_02, loc=<6.0104 0 0>, rot=<0 0 0.0149 -0.9999>}
 spine_04 {parent=spine_03, loc=<8.1024 0 0>, rot=<0 0 -0.0132 -0.9999>}
 spine_05 {parent=spine_04, loc=<10.0146 0 0>, rot=<0 0 -0.0015 -1>}
 neck_01 {parent=spine_05, loc=<11.3138 0.1529 0>, rot=<0 0 -0.2003 -0.9797>}
 neck_02 {parent=neck_01, loc=<4.5627 0 0>, rot=<0 0 0.0504 -0.9987>}
 head {parent=neck_02, loc=<5.3914 0 0>, rot=<0 0 0.115 -0.9934>}
 dyn_bk_r_hair_MSTR {parent=head, loc=<2.6204 -5.7536 7.0215>, rot=<0.9293 0.0651 -0.1816 -0.315>}
 dyn_bk_r_hair_01 {parent=dyn_bk_r_hair_MSTR, loc=<0 0 0>, rot=<0 0 0 -1>}
 dyn_bk_r_hair_02 {parent=dyn_bk_r_hair_01, loc=<-4.4045 0 0.0001>, rot=<0 0.0597 0 -0.9982>}
 dyn_bk_r_hair_03 {parent=dyn_bk_r_hair_02, loc=<-4.4048 0 0.0002>, rot=<0 0.0847 0 -0.9964>}
 dyn_bk_r_hair_04 {parent=dyn_bk_r_hair_03, loc=<-4.4044 0 0>, rot=<0 0.1378 0 -0.9905>}
 dyn_r_hairstrands_MSTR {parent=head, loc=<10.2998 5.4888 7.2022>, rot=<-0.5681 -0.1976 0.7869 0.138>}
 dyn_r_hairstrands_01 {parent=dyn_r_hairstrands_MSTR, loc=<0 0 0>, rot=<0 0 0 -1>}
 dyn_ft_r_hairstrands_MSTR {parent=head, loc=<11.2661 8.7207 1.1101>, rot=<-0.5681 -0.1976 0.7869 0.138>}
 dyn_ft_r_hairstrands_01 {parent=dyn_ft_r_hairstrands_MSTR, loc=<0 0 0>, rot=<0 0 0 -1>}
 dyn_r_sideburn_hair_MSTR {parent=head, loc=<4.5614 1.8805 8.4179>, rot=<1 0.0041 -0.0047 0>}
 dyn_r_sideburn_hair_01 {parent=dyn_r_sideburn_hair_MSTR, loc=<0 0 0>, rot=<0 0 0 1>}
 dyn_r_sideburn_hair_02 {parent=dyn_r_sideburn_hair_01, loc=<-4.4042 0 0>, rot=<0 0 0.0587 -0.9983>}
 dyn_r_sideburn_hair_03 {parent=dyn_r_sideburn_hair_02, loc=<-4.4045 0 0>, rot=<0 0 -0.0941 -0.9956>}
 dyn_r_sideburn_hair_04 {parent=dyn_r_sideburn_hair_03, loc=<-4.4043 0 0>, rot=<0 0 -0.1205 -0.9927>}
 dyn_bk_hair_MSTR {parent=head, loc=<5.1832 -9.0037 0>, rot=<0.0921 -0.7011 0.7011 -0.0921>}
 dyn_bk_hair_01 {parent=dyn_bk_hair_MSTR, loc=<0 0 0>, rot=<0 0 0 1>}
 dyn_bk_hair_02 {parent=dyn_bk_hair_01, loc=<4.4044 0 0>, rot=<0 0.0283 0 -0.9996>}
 dyn_bk_hair_03 {parent=dyn_bk_hair_02, loc=<4.4044 0 0>, rot=<0 0.044 0 -0.999>}
 dyn_bk_hair_04 {parent=dyn_bk_hair_03, loc=<4.4044 0 0>, rot=<0 0.0408 0 -0.9992>}
 dyn_bk_l_hair_MSTR {parent=head, loc=<2.6201 -5.7536 -7.0215>, rot=<0.0651 -0.9293 0.315 -0.1816>}
 dyn_bk_l_hair_01 {parent=dyn_bk_l_hair_MSTR, loc=<0 0 0>, rot=<0 0 0 1>}
 dyn_bk_l_hair_02 {parent=dyn_bk_l_hair_01, loc=<4.4044 0 0>, rot=<0 0.0597 0 -0.9982>}
 dyn_bk_l_hair_03 {parent=dyn_bk_l_hair_02, loc=<4.4044 0 0>, rot=<0 0.0847 0 -0.9964>}
 dyn_bk_l_hair_04 {parent=dyn_bk_l_hair_03, loc=<4.4044 0 0>, rot=<0 0.1378 0 -0.9905>}
 dyn_bk_tp_hair_MSTR {parent=head, loc=<12.3635 -3.0066 0>, rot=<0.4047 -0.5798 0.5798 -0.4047>}
 dyn_bk_tp_hair_01 {parent=dyn_bk_tp_hair_MSTR, loc=<0 0 0>, rot=<0 0 0 1>}
 dyn_bk_tp_hair_02 {parent=dyn_bk_tp_hair_01, loc=<9.214 0 0>, rot=<0 0.1533 0 -0.9882>}
 dyn_bk_tp_hair_03 {parent=dyn_bk_tp_hair_02, loc=<1.215 0 0>, rot=<0 0 0 1>}
 dyn_l_hairstrands_MSTR {parent=head, loc=<10.2997 5.4888 -7.2022>, rot=<0.1976 -0.5681 0.138 -0.7869>}
 dyn_l_hairstrands_01 {parent=dyn_l_hairstrands_MSTR, loc=<0 0 0>, rot=<0 0 0 1>}
 dyn_ft_l_hairstrands_MSTR {parent=head, loc=<11.266 8.7207 -1.1101>, rot=<0.1976 -0.5681 0.138 -0.7869>}
 dyn_ft_l_hairstrands_01 {parent=dyn_ft_l_hairstrands_MSTR, loc=<0 0 0>, rot=<0 0 0 1>}
 dyn_l_sideburn_hair_MSTR {parent=head, loc=<4.5613 1.8805 -8.4179>, rot=<0.0041 -1 0 -0.0047>}
 dyn_l_sideburn_hair_01 {parent=dyn_l_sideburn_hair_MSTR, loc=<0 0 0>, rot=<0 0 0 1>}
 dyn_l_sideburn_hair_02 {parent=dyn_l_sideburn_hair_01, loc=<4.4044 0 0>, rot=<0 0 0.0587 -0.9983>}
 dyn_l_sideburn_hair_03 {parent=dyn_l_sideburn_hair_02, loc=<4.4044 0 0>, rot=<0 0 -0.0941 -0.9956>}
 dyn_l_sideburn_hair_04 {parent=dyn_l_sideburn_hair_03, loc=<4.4044 0 0>, rot=<0 0 -0.1205 -0.9927>}
]
