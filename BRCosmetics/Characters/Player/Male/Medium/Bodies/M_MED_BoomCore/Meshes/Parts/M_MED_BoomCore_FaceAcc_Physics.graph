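# Physics asset for the BoomCore medium male FaceAcc mesh, covering physics for its back, top, sideburn and front/side strand hair.
# Each hair group is a capsule chain starting at a small MSTR body that joins the head with free swing and twist.
# Links within chains have narrow swing limits and one-degree twist limits.
# Head, neck, upper spine, clavicle and upper-arm bodies supply collision shapes for the upper body.
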PhysicsAsset /BRCosmetics/Characters/Player/Male/Medium/Bodies/M_MED_BoomCore/Meshes/Parts/M_MED_BoomCore_FaceAcc_Physics
bodies {
 clavicle_l [{shapeType=Capsule, length=9.0192, radius=9, center=<9.3294 0.1109 -2.4734>, rotation=<rot 69.3297 1.4476 176.9462>}]
 clavicle_r [{shapeType=Capsule, length=9.0192, radius=9, center=<-9.3294 -0.1109 2.4734>, rotation=<rot 69.3297 1.4476 -3.0538>}]
 dyn_bk_hair_01 [{shapeType=Capsule, length=15, radius=4, shapeName=dyn_bk_hair_01_capsule, center=<4 0 1>, rotation=<rot 0 0 90>}]
 dyn_bk_hair_02 [{shapeType=Capsule, length=15, radius=4, shapeName=dyn_bk_hair_02_capsule, center=<4 0 1>, rotation=<rot 0 0 90>}]
 dyn_bk_hair_03 [{shapeType=Capsule, length=15, radius=4, shapeName=dyn_bk_hair_03_capsule, center=<4 0 1>, rotation=<rot 0 0 90>}]
 dyn_bk_hair_04 [{shapeType=Capsule, length=15, radius=4, shapeName=dyn_bk_hair_04_capsule, center=<4 0 1>, rotation=<rot 0 0 90>}]
 dyn_bk_hair_MSTR [{shapeType=Capsule, length=0.505, radius=0.505, shapeName=dyn_bk_hair_MSTR_capsule, rotation=<rot 0 -90 0>}]
 dyn_bk_l_hair_01 [{shapeType=Capsule, length=15, radius=4, shapeName=dyn_bk_l_hair_01_capsule, center=<4 0 1>, rotation=<rot 0 0 90>}]
 dyn_bk_l_hair_02 [{shapeType=Capsule, length=15, radius=4, shapeName=dyn_bk_l_hair_02_capsule, center=<4 0 1>, rotation=<rot 0 0 90>}]
 dyn_bk_l_hair_03 [{shapeType=Capsule, length=15, radius=4, shapeName=dyn_bk_l_hair_03_capsule, center=<4 0 1>, rotation=<rot 0 0 90>}]
 dyn_bk_l_hair_04 [{shapeType=Capsule, length=15, radius=4, shapeName=dyn_bk_l_hair_04_capsule, center=<4 0 1>, rotation=<rot 0 0 90>}]
 dyn_bk_l_hair_MSTR [{shapeType=Capsule, length=0.505, radius=0.505, shapeName=dyn_bk_l_hair_MSTR_capsule, rotation=<rot 0 -90 0>}]
 dyn_bk_r_hair_01 [{shapeType=Capsule, length=15, radius=4, shapeName=dyn_bk_r_hair_01_capsule, center=<-4 0 -1>, rotation=<rot 0 0 90>}]
 dyn_bk_r_hair_02 [{shapeType=Capsule, length=15, radius=4, shapeName=dyn_bk_r_hair_02_capsule, center=<-4 0 -1>, rotation=<rot 0 0 90>}]
 dyn_bk_r_hair_03 [{shapeType=Capsule, length=15, radius=4, shapeName=dyn_bk_r_hair_03_capsule, center=<-4 0 -1>, rotation=<rot 0 0 90>}]
 dyn_bk_r_hair_04 [{shapeType=Capsule, length=15, radius=4, shapeName=dyn_bk_r_hair_04_capsule, center=<-4 0 -1>, rotation=<rot 0 0 90>}]
 dyn_bk_r_hair_MSTR [{shapeType=Capsule, length=0.505, radius=0.505, shapeName=dyn_bk_r_hair_MSTR_capsule, rotation=<rot 0 -90 0>}]
 dyn_bk_tp_hair_01 [{shapeType=Capsule, length=0, radius=6, shapeName=dyn_bk_tp_hair_01_capsule, center=<4 0 0>, rotation=<rot 0 90 90>}]
 dyn_bk_tp_hair_02 [{shapeType=Capsule, length=4, radius=1.5, shapeName=dyn_bk_tp_hair_02_capsule, center=<2 0 0>, rotation=<rot 0 90 90>}]
 dyn_bk_tp_hair_03 [{shapeType=Capsule, length=0, radius=8, shapeName=dyn_bk_tp_hair_03_capsule, center=<7.5 0 0>, rotation=<rot 0 90 90>}]
 dyn_bk_tp_hair_MSTR [{shapeType=Capsule, length=0.505, radius=0.505, shapeName=dyn_bk_tp_hair_MSTR_capsule, rotation=<rot 0 -90 0>}]
 dyn_ft_l_hairstrands_01 [{shapeType=Capsule, length=3, radius=2, shapeName=dyn_ft_l_hairstrands_01_capsule, center=<4 0 0>, rotation=<rot 0 -90 90>}]
 dyn_ft_l_hairstrands_MSTR [{shapeType=Capsule, length=0.505, radius=0.505, shapeName=dyn_ft_l_hairstrands_MSTR_capsule, rotation=<rot 0 -90 0>}]
 dyn_ft_r_hairstrands_01 [{shapeType=Capsule, length=3, radius=2, shapeName=dyn_ft_r_hairstrands_01_capsule, center=<-4 0 0>, rotation=<rot 0 -90 90>}]
 dyn_ft_r_hairstrands_MSTR [{shapeType=Capsule, length=0.505, radius=0.505, shapeName=dyn_ft_r_hairstrands_MSTR_capsule, rotation=<rot 0 -90 0>}]
 dyn_l_hairstrands_01 [{shapeType=Capsule, length=3, radius=2, shapeName=dyn_l_hairstrands_01_capsule, center=<4 0 0>, rotation=<rot 0 -90 90>}]
 dyn_l_hairstrands_MSTR [{shapeType=Capsule, length=0.505, radius=0.505, shapeName=dyn_l_hairstrands_MSTR_capsule, rotation=<rot 0 -90 0>}]
 dyn_l_sideburn_hair_01 [{shapeType=Capsule, length=3, radius=2.5, shapeName=dyn_l_sideburn_hair_01_capsule, center=<4 0 0>, rotation=<rot 0 90 90>}]
 dyn_l_sideburn_hair_02 [{shapeType=Capsule, length=3, radius=2.5, shapeName=dyn_l_sideburn_hair_02_capsule, center=<4 0 0>, rotation=<rot 0 90 90>}]
 dyn_l_sideburn_hair_03 [{shapeType=Capsule, length=3, radius=2.5, shapeName=dyn_l_sideburn_hair_03_capsule, center=<4 0 0>, rotation=<rot 0 130.0004 90>}]
 dyn_l_sideburn_hair_04 [{shapeType=Capsule, length=3, radius=2.5, shapeName=dyn_l_sideburn_hair_04_capsule, center=<4 3 0>, rotation=<rot 0 120.0002 90>}]
 dyn_l_sideburn_hair_MSTR [{shapeType=Capsule, length=0.505, radius=0.505, shapeName=dyn_l_sideburn_hair_MSTR_capsule, rotation=<rot 0 -90 0>}]
 dyn_r_hairstrands_01 [{shapeType=Capsule, length=3, radius=2, shapeName=dyn_r_hairstrands_01_capsule, center=<-4 0 0>, rotation=<rot 0 -90 90>}]
 dyn_r_hairstrands_MSTR [{shapeType=Capsule, length=0.505, radius=0.505, shapeName=dyn_r_hairstrands_MSTR_capsule, rotation=<rot 0 -90 0>}]
 dyn_r_sideburn_hair_01 [{shapeType=Capsule, length=3, radius=2.5, shapeName=dyn_r_sideburn_hair_01_capsule, center=<-4 0 0>, rotation=<rot 0 90 90>}]
 dyn_r_sideburn_hair_02 [{shapeType=Capsule, length=3, radius=2.5, shapeName=dyn_r_sideburn_hair_02_capsule, center=<-4 0 0>, rotation=<rot 0 90 90>}]
 dyn_r_sideburn_hair_03 [{shapeType=Capsule, length=3, radius=2.5, shapeName=dyn_r_sideburn_hair_03_capsule, center=<-4 0 0>, rotation=<rot 0 130.0004 90>}]
 dyn_r_sideburn_hair_04 [{shapeType=Capsule, length=3, radius=2.5, shapeName=dyn_r_sideburn_hair_04_capsule, center=<-4 -3 0>, rotation=<rot 0 120.0002 90>}]
 dyn_r_sideburn_hair_MSTR [{shapeType=Capsule, length=0.505, radius=0.505, shapeName=dyn_r_sideburn_hair_MSTR_capsule, rotation=<rot 0 -90 0>}]
 head [{shapeType=Sphere, radius=6.5, center=<3.3447 2.8474 0>}]
 neck_02 [{shapeType=Capsule, length=13.4375, radius=6, center=<-0.2642 -0.3408 0>, rotation=<rot 90 -171.1193 -175.3978>}]
 spine_03 [{shapeType=Capsule, length=8.2107, radius=12, center=<0.1445 1.1909 0.0001>}]
 spine_04 [{shapeType=Capsule, length=9.4404, radius=13, center=<2.766 2.5197 0.0002>}]
 spine_05 [
  {shapeType=Capsule, length=8.0226, radius=13, center=<-2.7491 -2.2521 0>}
  {shapeType=Capsule, length=13.9518, radius=11, center=<-0.2916 -1.4993 0>}
 ]
 upperarm_l [{shapeType=Capsule, length=20, radius=6, shapeName=upperarm_l_capsule, center=<11.7956 0 -0.77>, rotation=<rot 0 90 90>}]
 upperarm_r [{shapeType=Capsule, length=20, radius=6, shapeName=upperarm_r_capsule, center=<-11.7956 0 0.77>, rotation=<rot 0 90 90>}]
}
constraints [
 {
  bone1Name=dyn_bk_hair_MSTR
  bone2Name=head
  swing1LimitDegrees=45
  swing1Motion=Free
  swing2LimitDegrees=45
  swing2Motion=Free
  twistLimitDegrees=45
  twistMotion=Free
 }
 {bone1Name=dyn_bk_hair_01, bone2Name=dyn_bk_hair_MSTR, swing1LimitDegrees=5, swing2LimitDegrees=12, twistLimitDegrees=1}
 {bone1Name=dyn_bk_hair_02, bone2Name=dyn_bk_hair_01, swing1LimitDegrees=5, swing2LimitDegrees=12, twistLimitDegrees=1}
 {bone1Name=dyn_bk_hair_03, bone2Name=dyn_bk_hair_02, swing1LimitDegrees=5, swing2LimitDegrees=15, twistLimitDegrees=1}
 {bone1Name=dyn_bk_hair_04, bone2Name=dyn_bk_hair_03, swing1LimitDegrees=5, swing2LimitDegrees=15, twistLimitDegrees=1}
 {
  bone1Name=dyn_bk_l_hair_MSTR
  bone2Name=head
  swing1LimitDegrees=45
  swing1Motion=Free
  swing2LimitDegrees=45
  swing2Motion=Free
  twistLimitDegrees=45
  twistMotion=Free
 }
 {bone1Name=dyn_bk_l_hair_01, bone2Name=dyn_bk_l_hair_MSTR, swing1LimitDegrees=5, swing2LimitDegrees=12, twistLimitDegrees=1}
 {bone1Name=dyn_bk_l_hair_02, bone2Name=dyn_bk_l_hair_01, swing1LimitDegrees=5, swing2LimitDegrees=12, twistLimitDegrees=1}
 {bone1Name=dyn_bk_l_hair_03, bone2Name=dyn_bk_l_hair_02, swing1LimitDegrees=5, swing2LimitDegrees=15, twistLimitDegrees=1}
 {bone1Name=dyn_bk_l_hair_04, bone2Name=dyn_bk_l_hair_03, swing1LimitDegrees=5, swing2LimitDegrees=15, twistLimitDegrees=1}
 {
  bone1Name=dyn_bk_r_hair_MSTR
  bone2Name=head
  swing1LimitDegrees=45
  swing1Motion=Free
  swing2LimitDegrees=45
  swing2Motion=Free
  twistLimitDegrees=45
  twistMotion=Free
 }
 {bone1Name=dyn_bk_r_hair_01, bone2Name=dyn_bk_r_hair_MSTR, swing1LimitDegrees=5, swing2LimitDegrees=12, twistLimitDegrees=1}
 {bone1Name=dyn_bk_r_hair_02, bone2Name=dyn_bk_r_hair_01, swing1LimitDegrees=5, swing2LimitDegrees=12, twistLimitDegrees=1}
 {bone1Name=dyn_bk_r_hair_03, bone2Name=dyn_bk_r_hair_02, swing1LimitDegrees=5, swing2LimitDegrees=15, twistLimitDegrees=1}
 {bone1Name=dyn_bk_r_hair_04, bone2Name=dyn_bk_r_hair_03, swing1LimitDegrees=5, swing2LimitDegrees=15, twistLimitDegrees=1}
 {
  bone1Name=dyn_bk_tp_hair_MSTR
  bone2Name=head
  swing1LimitDegrees=45
  swing1Motion=Free
  swing2LimitDegrees=45
  swing2Motion=Free
  twistLimitDegrees=45
  twistMotion=Free
 }
 {bone1Name=dyn_bk_tp_hair_01, bone2Name=dyn_bk_tp_hair_MSTR, swing1LimitDegrees=5, swing2LimitDegrees=5, twistLimitDegrees=1}
 {bone1Name=dyn_bk_tp_hair_02, bone2Name=dyn_bk_tp_hair_01, swing1LimitDegrees=20, swing2LimitDegrees=20, twistLimitDegrees=1}
 {bone1Name=dyn_bk_tp_hair_03, bone2Name=dyn_bk_tp_hair_02, swing1LimitDegrees=20, swing2LimitDegrees=20, twistLimitDegrees=1}
 {
  bone1Name=dyn_ft_l_hairstrands_MSTR
  bone2Name=head
  swing1LimitDegrees=45
  swing1Motion=Free
  swing2LimitDegrees=45
  swing2Motion=Free
  twistLimitDegrees=45
  twistMotion=Free
 }
 {bone1Name=dyn_ft_l_hairstrands_01, bone2Name=dyn_ft_l_hairstrands_MSTR, swing1LimitDegrees=15, swing2LimitDegrees=15, twistLimitDegrees=1}
 {
  bone1Name=dyn_ft_r_hairstrands_MSTR
  bone2Name=head
  swing1LimitDegrees=45
  swing1Motion=Free
  swing2LimitDegrees=45
  swing2Motion=Free
  twistLimitDegrees=45
  twistMotion=Free
 }
 {bone1Name=dyn_ft_r_hairstrands_01, bone2Name=dyn_ft_r_hairstrands_MSTR, swing1LimitDegrees=15, swing2LimitDegrees=15, twistLimitDegrees=1}
 {
  bone1Name=dyn_l_hairstrands_MSTR
  bone2Name=head
  swing1LimitDegrees=45
  swing1Motion=Free
  swing2LimitDegrees=45
  swing2Motion=Free
  twistLimitDegrees=45
  twistMotion=Free
 }
 {bone1Name=dyn_l_hairstrands_01, bone2Name=dyn_l_hairstrands_MSTR, swing1LimitDegrees=15, swing2LimitDegrees=15, twistLimitDegrees=1}
 {
  bone1Name=dyn_l_sideburn_hair_MSTR
  bone2Name=head
  swing1LimitDegrees=45
  swing1Motion=Free
  swing2LimitDegrees=45
  swing2Motion=Free
  twistLimitDegrees=45
  twistMotion=Free
 }
 {bone1Name=dyn_l_sideburn_hair_01, bone2Name=dyn_l_sideburn_hair_MSTR, swing1LimitDegrees=15, swing2LimitDegrees=15, twistLimitDegrees=1}
 {bone1Name=dyn_l_sideburn_hair_02, bone2Name=dyn_l_sideburn_hair_01, swing1LimitDegrees=15, swing2LimitDegrees=15, twistLimitDegrees=1}
 {bone1Name=dyn_l_sideburn_hair_03, bone2Name=dyn_l_sideburn_hair_02, swing1LimitDegrees=15, swing2LimitDegrees=15, twistLimitDegrees=1}
 {bone1Name=dyn_l_sideburn_hair_04, bone2Name=dyn_l_sideburn_hair_03, swing1LimitDegrees=15, swing2LimitDegrees=15, twistLimitDegrees=1}
 {
  bone1Name=dyn_r_hairstrands_MSTR
  bone2Name=head
  swing1LimitDegrees=45
  swing1Motion=Free
  swing2LimitDegrees=45
  swing2Motion=Free
  twistLimitDegrees=45
  twistMotion=Free
 }
 {bone1Name=dyn_r_hairstrands_01, bone2Name=dyn_r_hairstrands_MSTR, swing1LimitDegrees=15, swing2LimitDegrees=15, twistLimitDegrees=1}
 {
  bone1Name=dyn_r_sideburn_hair_MSTR
  bone2Name=head
  swing1LimitDegrees=45
  swing1Motion=Free
  swing2LimitDegrees=45
  swing2Motion=Free
  twistLimitDegrees=45
  twistMotion=Free
 }
 {bone1Name=dyn_r_sideburn_hair_01, bone2Name=dyn_r_sideburn_hair_MSTR, swing1LimitDegrees=15, swing2LimitDegrees=15, twistLimitDegrees=1}
 {bone1Name=dyn_r_sideburn_hair_02, bone2Name=dyn_r_sideburn_hair_01, swing1LimitDegrees=15, swing2LimitDegrees=15, twistLimitDegrees=1}
 {bone1Name=dyn_r_sideburn_hair_03, bone2Name=dyn_r_sideburn_hair_02, swing1LimitDegrees=15, swing2LimitDegrees=15, twistLimitDegrees=1}
 {bone1Name=dyn_r_sideburn_hair_04, bone2Name=dyn_r_sideburn_hair_03, swing1LimitDegrees=15, swing2LimitDegrees=15, twistLimitDegrees=1}
]
